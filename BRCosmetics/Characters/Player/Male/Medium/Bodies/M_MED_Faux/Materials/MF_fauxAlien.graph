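# Material function for the Faux body's alien materials that drives a transformation effect through curve parameters, overriding opacity mask, world position offset and base color.
# Opacity blends from head and right-arm sphere masks on pre-skinned position to a gradient dissolve broken up by time-seeded cell noise.
# A remapped mask shrinks vertices inward and down, while the IsAlienHair switch fades color toward HairBlack along the dissolve.
# Several branches, including the left-arm mask, are unconnected.
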MaterialFunction /BRCosmetics/Characters/Player/Male/Medium/Bodies/M_MED_Faux/Materials/MF_fauxAlien
graph {nodes [
  n1 = PreSkinnedPosition()
  n2 = VertexInterpolator(VS=@n1)
  n3 = Multiply(
   A=NamedRerouteDeclaration(None=ShaderStageSwitch(PixelShader=@n2.PS, VertexShader=@n1)) {name=PreskinnedPos, nodeColor=<rgba 0 1 0.2969 1>}
   B=Constant3Vector() {constant=<rgba 1.5 1 1 0>}
  )
  n4 = SphereMask(A=@n3, B=Constant3Vector() {desc=top of head, constant=<rgba 0 0 165 0>}, Radius=Multiply(A=NamedRerouteUsage()) {constB=1.4})
  n5 = NamedRerouteUsage()
  n6 = Min(A=Add(A=Multiply(A=@n5)) {constB=10}) {constB=50, desc=stops the arms eating into the legs}
  Add(A=@n4, B=SphereMask(A=@n3, B=Constant3Vector() {desc=left arm, constant=<rgba 80 0 90 0>}, Radius=@n6))
  Frac(None=Time())
  n7 = GetMaterialAttributes(None=FunctionInput() {inputType=FunctionInput_MaterialAttributes}) {attributeGetTypes [#1]}
  n8 = VertexNormalWS()
  n9 = Constant3Vector() {desc=push down in height, constant=<rgba 0 0 -4 0>}
  n10 = MaterialFunctionCall(Input=NamedRerouteUsage(), Input Low=Constant() {r=110}, Input High=Constant() {r=150}, Target Low=Constant(), Target High=Constant() {r=1}) {materialFunction=&RemapValueRange}
  n11 = Multiply(A=Add(A=Multiply(A=@n8) {constB=-0.5}, B=@n9), B=SmoothStep(Value=Saturate(None=@n10.Result)))
  n12 = Constant3Vector() {desc=right arm, constant=<rgba -80 0 90 0>}
  n13 = Saturate(None=Add(A=@n4, B=SphereMask(A=@n3, B=@n12, Radius=@n6)))
  n14 = ChannelMaskParameter(None=Divide(A=NamedRerouteUsage(), B=PreSkinnedLocalBounds().Max)) {maskChannel=Blue, parameterName=GradientDirection, defaultValue=<rgba 0 0 1 0>}
  n15 = ScalarParameter() {parameterName=HT_L1_TransitionCurve}
  n16 = NamedRerouteUsage()
  n17 = ScalarParameter() {defaultValue=1, parameterName=Cell_Tiling}
  n18 = MaterialFunctionCall(Seed=Time(), Vec3=Ceil(None=Divide(A=@n16, B=@n17))) {materialFunction=&MF_Hash}
  n19 = Subtract(
   A=Subtract(A=Add(A=@n14.R), B=Multiply(A=@n15) {constB=2})
   B=Multiply(
    A=ChannelMaskParameter(None=@n18.Vec3) {group=SuperLevelling_Mask, parameterName=SL_TransitionNoiseChannel}.R
    B=ScalarParameter() {defaultValue=0.1, parameterName=NoiseStrength}
   )
  )
  n20 = Clamp(None=Round(None=@n19))
  n21 = ScalarParameter() {parameterName=JH_TransformCurve}
  n22 = LinearInterpolate(
   A=@n7.BaseColor
   B=LinearInterpolate(A=@n7.BaseColor, B=VectorParameter() {parameterName=HairBlack, defaultValue=<rgba 0 0 0 1>}.RGB, Alpha=Clamp(None=@n19))
   Alpha=@n21
  )
  FunctionOutput(None=SetMaterialAttributes!(
    MaterialAttributes=@n7.MaterialAttributes
    WorldPositionOffset=@n11
    OpacityMask=LinearInterpolate(A=@n13, B=@n20, Alpha=@n21)
    BaseColor=StaticSwitchParameter(True=@n22, False=@n7.BaseColor) {parameterName=IsAlienHair}
   ) {attributeSetTypes [#2, #3, #1]})
  n23 = VectorParameter() {parameterName=Diffuse Color, defaultValue=<rgba 0 0 0 1>}
  n24 = Constant3Vector() {constant=<rgba 1 0 0 0>}
  n25 = ScalarParameter() {parameterName=HT_L1_TransitionCurve}
  LinearInterpolate(
   A=LinearInterpolate(
    A=LinearInterpolate(A=@n23.RGB, B=@n24, Alpha=@n25)
    B=Constant3Vector() {constant=<rgba 0 1 0 0>}
    Alpha=ScalarParameter() {parameterName=JH_TransformCurve}
   )
   B=Constant3Vector() {constant=<rgba 0 0 1 0>}
   Alpha=ScalarParameter() {parameterName=Lexa_TransformationCurve}
  )
  LinearInterpolate(A=@n11, Alpha=@n21)
  Multiply(A=@n8) {constB=-1}
  n26 = Subtract(A=WorldPosition().XYZ, B=ObjectPositionWS())
  Multiply(A=Normalize(VectorInput=Transform(None=@n26) {transformSourceType=TRANSFORMSOURCE_World, transformType=TRANSFORM_Local}), B=@n9)
  NamedRerouteDeclaration(None=ScalarParameter() {desc=Alien SphereMask Radius, group=Faux_WPO, parameterName=MaterialCurveD}) {name=Radius, nodeColor=<rgba 0 1 0.0391 1>}
  NamedRerouteDeclaration(None=ScalarParameter() {desc=drives angle of bend - normalized, group=Faux_WPO, parameterName=MaterialCurveA}) {name=bendAmount, nodeColor=<rgba 1 0.9141 0 1>}
  NamedRerouteDeclaration(None=ScalarParameter() {desc=zipperHeight, group=Faux_WPO, parameterName=MaterialCurveB}) {name=zipperHeight, nodeColor=<rgba 0 1 0.6719 1>}
  ScalarParameter() {desc=drives angle of bend - normalized, group=Faux_WPO, parameterName=HT_L1_TransitionCurve}
  ScalarParameter() {desc=height of zipper, group=Faux_WPO, parameterName=JH_TransformCurve}
  ScalarParameter() {defaultValue=1000, group=Faux_WPO, parameterName=PeelyBlenderFace}
  ScalarParameter() {group=Faux_WPO, parameterName=Lexa_TransformationCurve}
  ScalarParameter() {desc=SphereMask Radius, group=Faux_WPO, parameterName=MaterialCurveC}
  SphereMask(A=@n19, B=Constant() {r=0.5}, Radius=Constant() {r=0.5}, Hardness=Constant())
  VectorParameter() {parameterName=Diffuse Color, defaultValue=<rgba 0.6 0.3 0.1 1>}
 ]}
