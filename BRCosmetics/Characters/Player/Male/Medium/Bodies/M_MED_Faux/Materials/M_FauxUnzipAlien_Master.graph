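# Master material for the male medium Faux Unzip Alien character body, using masked blending and material attributes on skeletal meshes.
# The PreFX character function feeds static-switched Clearcoat, Generic Gem and Metals LUT layers, all masked by one SkinFX_Mask texture, with selectable channels for clear coat and gem.
# The result then passes through the MF_fauxAlien and MF_EmissiveFresnel functions before the PostFX character function.
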Material /BRCosmetics/Characters/Player/Male/Medium/Bodies/M_MED_Faux/Materials/M_FauxUnzipAlien_Master
graph {
 properties {
  bUseMaterialAttributes=true
  bUsedWithMeshDeformer=true
  bUsedWithMorphTargets=true
  bUsedWithSkeletalMesh=true
  blendMode=BLEND_Masked
  pixelDepthOffsetMode=PDOM_Legacy
  usedShadingModels=Default Lit
 }
 nodes [
  n1 = MaterialFunctionCall() {materialFunction=&BaseFN_CharacterFunction_PreFX}
  n2 = TextureSample(Tex=TextureObjectParameter() {group=Skin FX, parameterName=SkinFX_Mask})
  n3 = StaticSwitchParameter(
   True=MaterialFunctionCall!(Alpha=StaticComponentMaskParameter(None=@n2.RGBA) {defaultR=true, group=Skin FX, parameterName=Clear Coat Channel}, In=@n1.Result) {materialFunction=/Game/Weapons/WeaponSkins/MaterialFunction/Clearcoat}.Result
   False=@n1.Result
  ) {group=Skin FX, parameterName=Use Clear Coat}
  n4 = StaticSwitchParameter(
   True=MaterialFunctionCall!(
    MASK=StaticComponentMaskParameter(None=@n2.RGBA) {defaultR=true, group=Skin FX, parameterName=Generic Gem Effect Channel}
    Color Change Sequence=StaticComponentMaskParameter(None=@n2.RGBA) {defaultB=true, group=Skin FX, parameterName=Generic Gem Color Change Channel}
    Material=@n3
   ) {materialFunction=/Game/Weapons/WeaponSkins/MaterialFunction/GenericGem}.Result
   False=@n3
  ) {group=Skin FX, parameterName=Use Generic Gem}
  n5 = StaticSwitchParameter(
   True=MaterialFunctionCall(Mask=@n2.RGBA, Materials=@n4) {materialFunction=&MF_MetalsLUT}.Result
   False=@n4
  ) {group=Skin FX, parameterName=UseMetalLUT}
  MaterialFunctionCall(
   AO=@n1.AO
   Distance Blend=@n1.AO
   Fresnel=@n1.AO
   Gmap=@n1.AO
   Pre Skinned Local Position=@n1.AO
   Material=MaterialFunctionCall!(In=MaterialFunctionCall!(In=@n5) {materialFunction=&/MF_fauxAlien}.Result) {materialFunction=/Game/Weapons/WeaponSkins/MaterialFunction/MF_EmissiveFresnel}.Result
  ) {materialFunction=&BaseFN_CharacterFunction_PostFX}
 ]
}
