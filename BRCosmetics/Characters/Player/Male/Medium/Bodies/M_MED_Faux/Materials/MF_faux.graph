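# Material function for the Faux player body that drives an animated peel-and-dissolve transformation on the incoming material attributes.
# Curve parameters drive bend, inflation and noisy wobble vertex offsets, with matching bent normals and sphere masks cutting away head and arms.
# JH_TransformCurve blends to a noise-edged gradient state that pulls vertices toward SocketLocation.
# Back faces get an inside color, and emissive adds seam glow or transition-edge glow, with procedural grid and hashed line patterns defined.
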MaterialFunction /BRCosmetics/Characters/Player/Male/Medium/Bodies/M_MED_Faux/Materials/MF_faux
graph {nodes [
  n1 = ScalarParameter() {defaultValue=1, parameterName=HT_L1_TransitionCurve}
  Ceil(None=@n1)
  n2 = NamedRerouteUsage()
  n3 = PreSkinnedLocalBounds()
  n4 = OneMinus(None=ChannelMaskParameter(None=Divide(A=@n2, B=@n3.Max)) {maskChannel=Blue, parameterName=GradientDirection, defaultValue=<rgba 0 0 1 0>}.R)
  n5 = OneMinus(None=@n1)
  n6 = NamedRerouteUsage()
  n7 = ScalarParameter() {defaultValue=1, parameterName=Cell_Tiling}
  n8 = MaterialFunctionCall(Seed=Time(), Vec3=Ceil(None=Divide(A=@n6, B=@n7))) {materialFunction=&MF_Hash}
  n9 = Subtract(
   A=Subtract(A=Add(A=@n4), B=Multiply(A=@n5) {constB=2})
   B=Multiply(
    A=ChannelMaskParameter(None=@n8.Vec3) {group=SuperLevelling_Mask, parameterName=SL_TransitionNoiseChannel}.R
    B=ScalarParameter() {defaultValue=0.1, parameterName=NoiseStrength}
   )
  )
  Clamp(None=@n9)
  ComponentMask(None=@n8.Vec3) {g=true}
  Frac(None=Time())
  n10 = GetMaterialAttributes(None=FunctionInput() {inputType=FunctionInput_MaterialAttributes}) {attributeGetTypes [#1, #2, #3]}
  n11 = Clamp(None=TwoSidedSign())
  n12 = MaterialFunctionCall!(
   angle=Multiply(A=NamedRerouteUsage()) {constB=110}
   Min=NamedRerouteUsage()
   PreskinnedPosition=NamedRerouteUsage()
   TexCoord=TextureCoordinate() {coordinateIndex=1, desc=coord 1 has the split uvs}
  ) {materialFunction=/Game/Characters/Player/Male/Medium/Bodies/M_MED_Faux/Materials/MF_bend_WPO}
  n13 = VertexNormalWS()
  n14 = PreSkinnedPosition()
  n15 = FunctionInput(Preview=ShaderStageSwitch(PixelShader=VertexInterpolator(VS=@n14).PS, VertexShader=@n14)) {bUsePreviewValueAsDefault=true, inputName=PreskinnedPosition}
  n16 = Constant3Vector() {desc=top of head, constant=<rgba 0 0 168 0>}
  n17 = SphereMask(A=@n15, B=@n16, Radius=Add(A=NamedRerouteUsage()) {constB=30}) {hardnessPercent=0}
  n18 = Add(
   A=Multiply(A=NamedRerouteUsage()) {constB=2, desc=global inflation}
   B=Multiply(A=Multiply(A=@n17) {constB=10, desc=spheremask inflation}, B=NamedRerouteUsage())
  )
  n19 = Constant()
  n20 = NamedRerouteUsage()
  n21 = Multiply(A=Multiply(A=ChannelMaskParameter(None=@n20) {maskChannel=Blue, defaultValue=<rgba 0 0 1 0>}.R, B=NamedRerouteUsage())) {constB=-0.01}
  n22 = Add(
   A=Add(A=@n12.Result, B=Multiply(A=@n13, B=@n18) {constB=5})
   B=MaterialFunctionCall(R=@n19, G=@n19, B=@n21) {materialFunction=&AppendMany}.RGB
  )
  n23 = TextureObjectParameter() {
   group=SuperLevelling_Mask
   parameterName=FX_WobbleNoise
   texture=/Game/Characters/Player/Male/Medium/Bodies/M_MED_Viper/Textures/Glitch/T_SoftNoise
  }
  n24 = VectorParameter() {group=SuperLevelling_Mask, parameterName=FX_WobbleNoise_Tiling, defaultValue=<rgba 128 128 32 1>}
  n25 = Add(
   A=NamedRerouteUsage()
   B=Multiply(A=Time(), B=VectorParameter() {parameterName=WobbleNoisePanSpeedAndDirection, defaultValue=<rgba 12 12 12 1>}.RGB)
  )
  n26 = PreSkinnedNormal()
  n27 = Subtract(A=Multiply(A=MaterialFunctionCall(TextureObject=@n23, TextureSize=@n24.RGB, WorldPosition=@n25, World Space Normal=@n26) {materialFunction=/Game/Weapons/WeaponSkins/Wraps/MaterialFunctions/WorldAlignedTexture}.XYZ Texture) {constB=2})
  n28 = VectorParameter() {parameterName=WobbleIntensity, defaultValue=<rgba 2 2 2 1>}
  n29 = NamedRerouteUsage()
  n30 = Multiply(A=@n27, B=Multiply(A=Multiply(A=@n28.RGB, B=@n29), B=NamedRerouteUsage()))
  n31 = VectorParameter() {parameterName=SocketLocation, defaultValue=<rgba 0 0 0 1>}
  n32 = WorldPosition()
  n33 = Subtract(A=@n31.RGB, B=@n32.XYZ)
  n34 = Clamp(None=Round(None=@n9))
  n35 = ScalarParameter() {parameterName=JH_TransformCurve}
  n36 = Multiply(A=NamedRerouteUsage()) {constB=110}
  n37 = NamedRerouteUsage()
  n38 = VertexNormalWS()
  n39 = TextureCoordinate() {coordinateIndex=1, desc=coord 1 has the split uvs}
  n40 = Normalize(VectorInput=Add(
    A=MaterialFunctionCall(angle=@n36, Min=@n37, PreskinnedPosition=@n38, TexCoord=@n39) {materialFunction=&MF_bend_WPO}.Result
    B=@n38
   ))
  n41 = Multiply(A=@n15, B=Constant3Vector() {constant=<rgba 1.5 1 1 0>})
  n42 = Multiply(A=NamedRerouteUsage()) {constB=1.1}
  n43 = TextureObjectParameter() {
   group=SuperLevelling_Mask
   parameterName=SL_TransitionNoise
   texture=&T_SoftNoise
  }
  n44 = MaterialFunctionCall(
   TextureObject=@n43
   TextureSize=VectorParameter() {group=SuperLevelling_Mask, parameterName=SL_TransitionNoise_Tiling, defaultValue=<rgba 6 30 100 1>}.RGB
   WorldPosition=@n15
   World Space Normal=VertexInterpolator(VS=PreSkinnedNormal()).PS
  ) {materialFunction=&WorldAlignedTexture}
  n45 = Multiply(A=Subtract(A=ChannelMaskParameter(None=@n44.XYZ Texture) {group=SuperLevelling_Mask, parameterName=SL_TransitionNoiseChannel}.R) {constB=0.5}) {constB=20}
  n46 = NamedRerouteUsage()
  n47 = Min(A=Add(A=Multiply(A=@n46)) {constB=5}) {constB=50, desc=stops the arms eating into the legs}
  n48 = Add(
   A=SphereMask(A=@n41, B=@n16, Radius=Add(A=@n42, B=@n45))
   B=SphereMask(A=@n41, B=Constant3Vector() {desc=left arm, constant=<rgba 80 0 90 0>}, Radius=Add(A=@n47, B=@n45))
  )
  n49 = Multiply(A=NamedRerouteUsage()) {constB=80}
  n50 = SphereMask(
   A=@n15
   B=Constant3Vector() {desc=right arm, constant=<rgba -80 0 90 0>}
   Radius=Add(A=Min(A=@n49) {constB=50, desc=stops the arms eating into the legs}, B=@n45)
  )
  n51 = OneMinus(None=Saturate(None=Add(A=@n48, B=@n50)))
  n52 = Ceil(None=NamedRerouteUsage())
  n53 = LinearInterpolate(B=OneMinus(None=Clamp(None=@n52)), Alpha=VertexColor().R) {constA=1}
  n54 = ScalarParameter() {defaultValue=128, parameterName=SprayMaskRAdius}
  n55 = OneMinus(None=SphereMask(A=@n32.XYZ, B=@n31.RGB, Radius=Multiply(A=@n54, B=@n5), Hardness=ScalarParameter() {parameterName=SprayMaskHardness}))
  n56 = VectorParameter() {parameterName=Emissive_Color, defaultValue=<rgba 0 0.5905 1 1>}
  n57 = Multiply(
   A=SphereMask(
    A=NamedRerouteUsage()
    B=ScalarParameter() {parameterName=GlowFadeSamplePos}
    Radius=ScalarParameter() {defaultValue=0.5, parameterName=GlowFadeRadius}
    Hardness=ScalarParameter() {parameterName=GlowfadeHardness}
   )
   B=SphereMask(
    A=VertexColor().G
    B=Constant() {r=1}
    Radius=ScalarParameter() {defaultValue=0.25, parameterName=TearEdgeRadius}
    Hardness=ScalarParameter() {parameterName=TearEdgeHardness}
   )
  )
  n58 = Power(Base=Multiply(A=ScalarParameter() {defaultValue=10, parameterName=SeamEmissive_Strength}, B=Multiply(A=@n56.RGB, B=@n57)))
  n59 = NamedRerouteUsage()
  n60 = SphereMask(A=@n9, B=Constant() {r=0.5}, Radius=Constant() {r=0.5}, Hardness=Constant())
  n61 = NamedRerouteUsage()
  n62 = Power(Base=Multiply(A=Multiply(A=@n61, B=@n56.RGB), B=ScalarParameter() {defaultValue=5, parameterName=Emissive_Strength}))
  n63 = Add(A=@n10.EmissiveColor, B=LinearInterpolate(A=Multiply(A=@n58, B=@n59), B=Multiply(A=@n60, B=@n62), Alpha=@n35))
  n64 = SetMaterialAttributes!(
   MaterialAttributes=@n10.MaterialAttributes
   BaseColor=LinearInterpolate(A=VectorParameter() {parameterName=InsideDiffuse Color, defaultValue=<rgba 0.25 0.142 0.223 1>}.RGB, B=@n10.BaseColor, Alpha=@n11)
   WorldPositionOffset=LinearInterpolate(A=Add(A=@n22, B=@n30), B=LinearInterpolate(A=@n33, Alpha=@n34) {constB=0}, Alpha=@n35) {constB=0}
   OpacityMask=LinearInterpolate(A=Multiply(A=@n51, B=@n53), B=Multiply(A=@n34, B=@n55), Alpha=@n35)
   EmissiveColor=@n63
   Normal=MaterialFunctionCall(BaseNormal=Transform(None=@n40) {transformSourceType=TRANSFORMSOURCE_World, transformType=TRANSFORM_Tangent}, AdditionalNormal=@n10.Normal) {materialFunction=&BlendAngleCorrectedNormals}.Result
  ) {attributeSetTypes [
    #3
    #4
    #5
    #1
    #2
   ]}
  FunctionOutput(None=@n64)
  n65 = ScalarParameter() {defaultValue=5, parameterName=Emissive_Strength}
  n66 = Multiply(A=VectorParameter() {parameterName=Emissive_Color, defaultValue=<rgba 0 0.5905 1 1>}.RGB, B=NamedRerouteUsage())
  LinearInterpolate(A=Power(Base=Multiply(A=@n65, B=@n66)), B=@n63, Alpha=@n11)
  n67 = VectorParameter() {parameterName=VariationIntensity, defaultValue=<rgba 0.25 0.25 0.25 1>}
  LinearInterpolate(A=@n33, B=Add(A=Multiply(A=@n27, B=@n67.RGB), B=Subtract(A=@n32.XYZ, B=@n31.RGB)), Alpha=@n55)
  n68 = VectorParameter() {parameterName=Diffuse Color, defaultValue=<rgba 0 0 0 1>}
  n69 = Constant3Vector() {constant=<rgba 1 0 0 0>}
  LinearInterpolate(
   A=LinearInterpolate(A=LinearInterpolate(A=@n68.RGB, B=@n69), B=Constant3Vector() {constant=<rgba 0 1 0 0>})
   B=Constant3Vector() {constant=<rgba 0 0 1 0>}
  )
  MaterialFunctionCall() {materialFunction=&MF_Hash}
  MaterialFunctionCall!(Rotation Axis, Rotation Angle, World Space Vertex Normals, Movement Mask, PivotPoint) {materialFunction=&FixRotateAboutAxisNormals}
  Multiply(A=@n21, B=NamedRerouteUsage())
  Multiply()
  NamedRerouteDeclaration(None=ScalarParameter() {desc=SphereMask Radius, group=Faux_WPO, parameterName=MaterialCurveC}) {name=Radius, nodeColor=<rgba 0 1 0.0391 1>}
  NamedRerouteDeclaration(None=ScalarParameter() {desc=drives angle of bend - normalized, group=Faux_WPO, parameterName=MaterialCurveA}) {name=BendAmt, nodeColor=<rgba 0.2656 1 0 1>}
  NamedRerouteDeclaration(None=ScalarParameter() {defaultValue=170, desc=zipperHeight, group=Faux_WPO, parameterName=MaterialCurveB}) {name=zipperHeight, nodeColor=<rgba 0 1 0.6719 1>}
  NamedRerouteDeclaration(None=@n15) {name=PreskinnedPos, nodeColor=<rgba 1 0 0.3281 1>}
  n70 = NamedRerouteUsage()
  n71 = ScalarParameter() {defaultValue=3, parameterName=Pattern_Tiling}
  n72 = MaterialFunctionCall(Seed=Time(), Vec3=Ceil(None=Divide(A=@n70, B=@n71))) {materialFunction=&MF_Hash}
  n73 = Frac(None=Divide(A=NamedRerouteUsage(), B=@n71))
  n74 = PreSkinnedNormal()
  n75 = ShaderStageSwitch(PixelShader=VertexInterpolator(VS=@n74).PS, VertexShader=@n74)
  n76 = ComponentMask(None=@n75) {r=true}
  n77 = LinearInterpolate(
   A=LinearInterpolate(A=ComponentMask(None=@n73) {g=true, r=true}, B=ComponentMask(None=@n73) {b=true, g=true}, Alpha=Abs(None=@n76))
   B=ComponentMask(None=@n73) {b=true, r=true}
   Alpha=Abs(None=ComponentMask(None=@n75) {g=true})
  )
  n78 = MaterialFunctionCall!(Angle=Clamp(None=ComponentMask(None=@n72.Vec3) {r=true}) {minDefault=-1}, UV/2DVector=Subtract(A=Multiply(A=@n77) {constB=2})) {materialFunction=/Game/Characters/Player/Male/Medium/Bodies/M_MED_Faux/Materials/MF_2DRotate}
  n79 = Add(A=@n78.Result, B=Constant2Vector() {g=-1})
  n80 = Add(A=@n78.Result, B=Constant2Vector() {g=1})
  n81 = Max(
   A=SphereMask(
    A=DotProduct(A=@n79, B=@n79)
    B=Constant() {r=0.05}
    Radius=ScalarParameter() {defaultValue=0.025, parameterName=LineThickness}
    Hardness=ScalarParameter() {defaultValue=1, parameterName=LineHardness}
   )
   B=SphereMask(
    A=DotProduct(A=@n80, B=@n80)
    B=Constant() {r=0.05}
    Radius=ScalarParameter() {defaultValue=0.025, parameterName=LineThickness}
    Hardness=ScalarParameter() {defaultValue=1, parameterName=LineHardness}
   )
  )
  n82 = SphereMask(
   A=ComponentMask(None=@n78.Result) {r=true}
   B=Constant()
   Radius=ScalarParameter() {defaultValue=0.025, parameterName=LineThickness}
   Hardness=ScalarParameter() {defaultValue=1, parameterName=LineHardness}
  )
  NamedRerouteDeclaration(None=Multiply(A=Max(A=@n81, B=@n82), B=ComponentMask(None=@n72.Vec3) {g=true})) {name=NoisePattern_01, nodeColor=<rgba 0.8047 0 1 1>}
  NamedRerouteDeclaration(None=@n12.ZMinusmin) {name=BendPoint, nodeColor=<rgba 0.4062 1 0 1>}
  n83 = Frac(None=Divide(A=NamedRerouteUsage(), B=ScalarParameter() {defaultValue=7, parameterName=InnerFaceGridResolution}))
  n84 = SphereMask(A=ComponentMask(None=@n83) {r=true}, B=Constant() {r=0.5}, Radius=Constant() {r=0.025}, Hardness=Constant())
  n85 = SphereMask(A=ComponentMask(None=@n83) {g=true}, B=Constant() {r=0.5}, Radius=Constant() {r=0.025}, Hardness=Constant())
  n86 = ComponentMask(None=@n83) {b=true}
  n87 = Constant() {r=0.5}
  n88 = Constant() {r=0.025}
  n89 = Constant()
  NamedRerouteDeclaration(None=Max(A=Max(A=@n84, B=@n85), B=SphereMask(A=@n86, B=@n87, Radius=@n88, Hardness=@n89))) {name=GridPattern, nodeColor=<rgba 1 0.2344 0 1>}
  n90 = NamedRerouteUsage()
  n91 = Multiply(A=ComponentMask(None=PreSkinnedLocalBounds().Max) {b=true}, B=ScalarParameter() {defaultValue=0.25, parameterName=Falloff})
  NamedRerouteDeclaration(None=Clamp(None=Divide(A=@n90, B=@n91))) {name=ZipperGradient, nodeColor=<rgba 0 1 0.0156 1>}
  NamedRerouteUsage()
  NamedRerouteUsage()
  PreSkinnedPosition()
  ScalarParameter() {defaultValue=1, parameterName=Cell_Tiling}
 ]}
