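# Masked, two-sided master material for the Faux player character body (FauxUnzip), built on material attributes for skeletal meshes with morph targets and mesh deformers.
# The shared character PreFX function feeds optional Silk and Thin Film layers, each behind a static switch and masked by a selectable channel of the SkinFX_Mask texture.
# The result passes through MF_faux, which also receives pre-skinned local position, then into the shared character PostFX function.
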Material /BRCosmetics/Characters/Player/Male/Medium/Bodies/M_MED_Faux/Materials/M_FauxUnzip_Master
graph {
 properties {
  bUseMaterialAttributes=true
  bUsedWithMeshDeformer=true
  bUsedWithMorphTargets=true
  bUsedWithSkeletalMesh=true
  blendMode=BLEND_Masked
  pixelDepthOffsetMode=PDOM_Legacy
  twoSided=true
  usedShadingModels=Default Lit
 }
 nodes [
  n1 = MaterialFunctionCall() {materialFunction=&BaseFN_CharacterFunction_PreFX}
  n2 = TextureSample(Tex=TextureObjectParameter() {group=Skin FX, parameterName=SkinFX_Mask})
  n3 = StaticSwitchParameter(
   True=MaterialFunctionCall!(Alpha=ChannelMaskParameter(None=@n2.RGBA) {group=Skin FX, parameterName=Silk Channel}.R, Material=@n1.Result) {materialFunction=/Game/Weapons/WeaponSkins/MaterialFunction/MF_Silk}.Result
   False=@n1.Result
  ) {group=Skin FX, parameterName=Use Silk}
  n4 = MaterialFunctionCall!(Alpha=StaticComponentMaskParameter(None=@n2.RGBA) {defaultR=true, group=Skin FX, parameterName=ThinFilm_Channel}, Material Attributes=@n3) {materialFunction=/Game/Weapons/WeaponSkins/MaterialFunction/ThinFilmFunction}
  MaterialFunctionCall(
   AO=@n1.AO
   Distance Blend=@n1.AO
   Fresnel=@n1.AO
   Gmap=@n1.AO
   Pre Skinned Local Position=@n1.AO
   Material=MaterialFunctionCall!(
    PreskinnedPosition=@n1.Pre skinned Local Position
    In=StaticSwitchParameter(True=@n4.Result, False=@n3) {group=Skin FX, parameterName=Use Thin Film}
   ) {materialFunction=&/MF_faux}.Result
  ) {materialFunction=&BaseFN_CharacterFunction_PostFX}
 ]
}
