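# Post-process style animation Blueprint for the M_MED_Faux male body part that adjusts attachment and accessory bones.
# AnimGraph copies the attached parent's pose and curves, then chains weighted constraints, additive bone offsets and one Look At to place backpack, cape, belt, buckle and pack bones.
# EventGraph derives thigh and right-arm alphas from bone deltas, plus crouch and front-end idle values, though the AnimGraph shows no reads of them.
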AnimBlueprint /BRCosmetics/Characters/Player/Male/Medium/Bodies/M_MED_Faux/Meshes/M_MED_Faux_AnimBP
target_skeleton=&/M_MED_Faux_Skeleton
graphs [
 AnimGraph {nodes [
   n1 = "Constraint - attach_backpack (PPPPP)"!(
    ComponentPose=Local To Component(LocalPose=Copy Pose From Mesh() {bCopyCurves=true, bUseAttachedParent=true})
    Alpha=1
    ConstraintWeights_0=0.016
    ConstraintWeights_1:float=0.1401
    ConstraintWeights_2:float=0.2032
    ConstraintWeights_3:float=0.6238
    ConstraintWeights_4:float=0.0168
   ) -> (Pose) {
    alphaBoolBlend {blendOption=Linear}
    boneToModify {boneName=attach_backpack}
    constraintSetup [
     {targetBone {boneName=clavicle_r}}
     {targetBone {boneName=spine_03}}
     {targetBone {boneName=spine_04}}
     {targetBone {boneName=spine_05}}
     {targetBone {boneName=clavicle_l}}
    ]
    constraintWeights [1,1,1,1,1]
   }
   n2 = "Transform (Modify) Bone - Bone: attach_backpack"(ComponentPose=@n1, Alpha=1, Translation="2.8,0,1", Rotation="0,0,0", Scale="1,1,1") {
    translationMode=BMM_Additive
    translationSpace=BCS_BoneSpace
    alphaBoolBlend {blendOption=Linear}
    boneToModify {boneName=attach_backpack}
    translation {x=2.8, z=1}
   }
   n3 = "Constraint - attach_cape (PPPPP)"!(
    ComponentPose=@n2
    Alpha=1
    ConstraintWeights_0=0.016
    ConstraintWeights_1:float=0.1401
    ConstraintWeights_2:float=0.2032
    ConstraintWeights_3:float=0.6238
    ConstraintWeights_4:float=0.0168
   ) -> (Pose) {
    alphaBoolBlend {blendOption=Linear}
    boneToModify {boneName=attach_cape}
    constraintSetup [
     {targetBone {boneName=clavicle_r}}
     {targetBone {boneName=spine_03}}
     {targetBone {boneName=spine_04}}
     {targetBone {boneName=spine_05}}
     {targetBone {boneName=clavicle_l}}
    ]
    constraintWeights [1,1,1,1,1]
   }
   n4 = "Transform (Modify) Bone - Bone: attach_cape"(ComponentPose=@n3, Alpha=1, Translation="2.8,0,1", Rotation="0,0,0", Scale="1,1,1") {
    translationMode=BMM_Additive
    translationSpace=BCS_BoneSpace
    alphaBoolBlend {blendOption=Linear}
    boneToModify {boneName=attach_cape}
    translation {x=2.8, z=1}
   }
   n5 = "Transform (Modify) Bone - Bone: dyn_parent_chestpackA"(ComponentPose=@n4, Alpha=1, Translation="0.000,-5,-0.000", Rotation="0,0,0", Scale="1,1,1") {
    rotationMode=BMM_Additive
    rotationSpace=BCS_BoneSpace
    translationMode=BMM_Additive
    translationSpace=BCS_BoneSpace
    alphaBoolBlend {blendOption=Linear}
    alphaScaleBiasClamp {interpSpeedDecreasing=15, interpSpeedIncreasing=15}
    boneToModify {boneName=dyn_parent_chestpackA}
    translation {y=-5}
   }
   n6 = "Transform (Modify) Bone - Bone: dyn_parent_belt_frt_r"(ComponentPose=@n5, Alpha=1, Translation="0,0,0", Rotation="20,0,20", Scale="1,1,1") {
    rotationMode=BMM_Additive
    rotationSpace=BCS_BoneSpace
    translationMode=BMM_Additive
    translationSpace=BCS_BoneSpace
    alphaBoolBlend {blendOption=Linear}
    boneToModify {boneName=dyn_parent_belt_frt_r}
    rotation {pitch=20, roll=20}
   }
   n7 = "Transform (Modify) Bone - Bone: dyn_belt_frt_r"(ComponentPose=@n6, Alpha=1, Translation="0.000,-1.000,3.000", Rotation="-5.000,0.000,0.000", Scale="1,1,1") {
    rotationMode=BMM_Additive
    rotationSpace=BCS_BoneSpace
    translationMode=BMM_Additive
    translationSpace=BCS_BoneSpace
    alphaBoolBlend {blendOption=Linear}
    boneToModify {boneName=dyn_belt_frt_r}
    rotation {pitch=-5}
    translation {y=-1, z=3}
   }
   n8 = "Transform (Modify) Bone - Bone: dyn_parent_belt_r"(ComponentPose=@n7, Alpha=1, Translation="0,0,0", Rotation="0,15,0", Scale="1,1,1") {
    rotationMode=BMM_Additive
    rotationSpace=BCS_BoneSpace
    translationMode=BMM_Additive
    translationSpace=BCS_BoneSpace
    alphaBoolBlend {blendOption=Linear}
    boneToModify {boneName=dyn_parent_belt_r}
    rotation {yaw=15}
   }
   n9 = "Transform (Modify) Bone - Bone: dyn_belt_r"(ComponentPose=@n8, Alpha=1, Translation="0,3,0", Rotation="-15,10,0", Scale="1,1,1") {
    rotationMode=BMM_Additive
    rotationSpace=BCS_BoneSpace
    translationMode=BMM_Additive
    translationSpace=BCS_BoneSpace
    alphaBoolBlend {blendOption=Linear}
    boneToModify {boneName=dyn_belt_r}
    rotation {pitch=-15, yaw=10}
    translation {y=3}
   }
   n10 = "Transform (Modify) Bone - Bone: dyn_parent_buckle"(ComponentPose=@n9, Alpha=1, Translation="0,1,-2", Rotation="-8,3,0.000", Scale="1,1,1") {
    rotationMode=BMM_Additive
    rotationSpace=BCS_BoneSpace
    translationMode=BMM_Additive
    translationSpace=BCS_BoneSpace
    alphaBoolBlend {blendOption=Linear}
    boneToModify {boneName=dyn_parent_buckle}
    rotation {pitch=-8, yaw=3}
    translation {y=1, z=-2}
   }
   n11 = "Transform (Modify) Bone - Bone: dyn_parent_belt_r"(ComponentPose=@n10, Alpha=1, Translation="0,0,0", Rotation="20,0.000,14", Scale="1,1,1") {
    rotationMode=BMM_Additive
    rotationSpace=BCS_BoneSpace
    translationMode=BMM_Additive
    translationSpace=BCS_BoneSpace
    alphaBoolBlend {blendOption=Linear}
    boneToModify {boneName=dyn_parent_belt_r}
    rotation {pitch=20, roll=14}
   }
   n12 = "Transform (Modify) Bone - Bone: dyn_belt_r"(ComponentPose=@n11, Alpha=1, Translation="-2,3,0", Rotation="-15,10,0", Scale="1,1,1") {
    rotationMode=BMM_Additive
    rotationSpace=BCS_BoneSpace
    translationMode=BMM_Additive
    translationSpace=BCS_BoneSpace
    alphaBoolBlend {blendOption=Linear}
    boneToModify {boneName=dyn_belt_r}
    rotation {pitch=-15, yaw=10}
    translation=<-2 3>
   }
   n13 = "Constraint - dyn_parent_pack_r (PPPPP)"!(
    ComponentPose=@n12
    Alpha=0.6
    ConstraintWeights_0=0.1227
    ConstraintWeights_1:float=0.0368
    ConstraintWeights_2:float=0.1811
    ConstraintWeights_3:float=0.6469
    ConstraintWeights_4:float=0.0125
   ) -> (Pose) {
    alphaBoolBlend {blendOption=Linear}
    boneToModify {boneName=dyn_parent_pack_r}
    constraintSetup [
     {targetBone {boneName=pelvis}}
     {targetBone {boneName=dyn_belt_frt_r}}
     {targetBone {boneName=thigh_twist_01_r}}
     {targetBone {boneName=dyn_belt_r}}
     {targetBone {boneName=spine_01}}
    ]
    constraintWeights [1,1,1,1,1]
   }
   n14 = Look At - Bone: dyn_pack_r(ComponentPose=@n13, Alpha=1) {
    alphaBoolBlend {blendOption=Linear}
    boneToModify {boneName=dyn_pack_r}
    lookAtTarget {boneReference {boneName=dyn_pack_aim_r}}
    lookAt_Axis {axis=<1 0>}
   }
   n15 = "Transform (Modify) Bone - Bone: dyn_parent_packA_l"(ComponentPose=@n14, Alpha=1, Translation="0.004,-4.695,-0.768", Rotation="-11,5,-34.841", Scale="1,1,1") {
    rotationMode=BMM_Additive
    rotationSpace=BCS_BoneSpace
    translationMode=BMM_Additive
    translationSpace=BCS_BoneSpace
    alphaBoolBlend {blendOption=Linear}
    boneToModify {boneName=dyn_parent_packA_l}
    rotation=<rot -11 5 -34.841>
    translation=<0.004 -4.695 -0.768>
   }
   n16 = "Transform (Modify) Bone - Bone: dyn_parent_packB_l"(ComponentPose=@n15, Alpha=1, Translation="0.2,-3.55,-0.8", Rotation="-8,0,-25", Scale="1,1,1") {
    rotationMode=BMM_Additive
    rotationSpace=BCS_BoneSpace
    translationMode=BMM_Additive
    translationSpace=BCS_BoneSpace
    alphaBoolBlend {blendOption=Linear}
    boneToModify {boneName=dyn_parent_packB_l}
    rotation {pitch=-8, roll=-25}
    translation=<0.2 -3.55 -0.8>
   }
   n17 = "Transform (Modify) Bone - Bone: dyn_parent_packA_l"(ComponentPose=@n16, Alpha=1, Translation="0,0,0", Rotation="24,-2,-13", Scale="1,1,1") {
    rotationMode=BMM_Additive
    rotationSpace=BCS_BoneSpace
    translationMode=BMM_Additive
    translationSpace=BCS_BoneSpace
    alphaBoolBlend {blendOption=Linear}
    alphaScaleBiasClamp {bInterpResult=true, interpSpeedDecreasing=16, interpSpeedIncreasing=16}
    boneToModify {boneName=dyn_parent_packA_l}
    rotation=<rot 24 -2 -13>
   }
   n18 = "Transform (Modify) Bone - Bone: dyn_parent_packB_l"(ComponentPose=@n17, Alpha=1, Translation="0,0,0", Rotation="24.141,0,0", Scale="1,1,1") {
    rotationMode=BMM_Additive
    rotationSpace=BCS_BoneSpace
    translationMode=BMM_Additive
    translationSpace=BCS_BoneSpace
    alphaBoolBlend {blendOption=Linear}
    alphaScaleBiasClamp {bInterpResult=true, interpSpeedDecreasing=16, interpSpeedIncreasing=16}
    boneToModify {boneName=dyn_parent_packB_l}
    rotation {pitch=24.141}
   }
   n19 = "Transform (Modify) Bone - Bone: dyn_parent_packA_l"(ComponentPose=@n18, Alpha=1, Translation="0,0,0", Rotation="5.074,-1.466,-29.665", Scale="1,1,1") {
    rotationMode=BMM_Additive
    rotationSpace=BCS_BoneSpace
    translationMode=BMM_Additive
    translationSpace=BCS_BoneSpace
    alphaBoolBlend {blendOption=Linear}
    alphaScaleBiasClamp {bInterpResult=true, interpSpeedDecreasing=16, interpSpeedIncreasing=16}
    boneToModify {boneName=dyn_parent_packA_l}
    rotation=<rot 5.074 -1.466 -29.665>
   }
   n20 = "Constraint - dyn_parent_chestpackB (PPPPP)"!(
    ComponentPose=@n19
    Alpha=1
    ConstraintWeights_0=0.4645
    ConstraintWeights_1:float=0.2279
    ConstraintWeights_2:float=0.1931
    ConstraintWeights_3:float=0.0725
    ConstraintWeights_4:float=0.042
   ) -> (Pose) {
    alphaBoolBlend {blendOption=Linear}
    boneToModify {boneName=dyn_parent_chestpackB}
    constraintSetup [
     {targetBone {boneName=spine_04}}
     {targetBone {boneName=spine_05}}
     {targetBone {boneName=spine_03}}
     {targetBone {boneName=clavicle_r}}
     {targetBone {boneName=spine_02}}
    ]
    constraintWeights [1,1,1,1,1]
   }
   n21 = "Transform (Modify) Bone - Bone: dyn_parent_chestpackB"(ComponentPose=@n20, Alpha=1, Translation="0,0,-0.25", Rotation="0,0,0", Scale="1,1,1") {
    rotationMode=BMM_Additive
    rotationSpace=BCS_BoneSpace
    translationMode=BMM_Additive
    translationSpace=BCS_BoneSpace
    alphaBoolBlend {blendOption=Linear}
    boneToModify {boneName=dyn_parent_chestpackB}
    translation {z=-0.25}
   }
   Output Pose(Result=Component To Local(ComponentPose=@n21)) {layerGroup=None}
  ]}
 EventGraph {nodes [
   Event BlueprintUpdateAnimation()
   > Get!(execute:Exec) -> (then:Exec, else:Exec, Skeletal_Mesh:Skeletal Mesh Component Object Reference)
   > n1 = Sequence!(execute) -> (then_0, then_1, then_2:Exec, then_3:Exec)
   > Set thigh_rot_fwd_r(thigh_rot_fwd_r:double=GetAlphaFromReferenceTransformDelta(
     SkeletalMeshComponent=Get Skeletal_Mesh() -> (Skeletal_Mesh:Skeletal Mesh Component Object Reference)
     BoneName=thigh_r
     BaseName=pelvis
     DeltaComponentType=QuaternionTwist
     InputRange=(X=-10,Y=50)
     TwistAxis="0,0,1"
    )) -> (Output_Get:double)
   > Set thigh_rot_out_r(thigh_rot_out_r:double=GetAlphaFromReferenceTransformDelta(
     SkeletalMeshComponent=Get Skeletal_Mesh()
     BoneName=thigh_r
     BaseName=pelvis
     DeltaComponentType=EulerY
     InputRange=(X=0,Y=40)
     TwistAxis="0,1,0"
    )) -> (Output_Get:double)
   Set thigh_rot_fwd_l(
    execute=@n1.then_1
    thigh_rot_fwd_l:double=GetAlphaFromReferenceTransformDelta(
     SkeletalMeshComponent=Get Skeletal_Mesh()
     BoneName=thigh_l
     BaseName=pelvis
     DeltaComponentType=QuaternionTwist
     InputRange=(X=0,Y=60)
     TwistAxis="0,0,1"
    )
   ) -> (Output_Get:double)
   > Set thigh_rot_out_l(thigh_rot_out_l:double=GetAlphaFromReferenceTransformDelta(
     SkeletalMeshComponent=Get Skeletal_Mesh()
     BoneName=thigh_l
     BaseName=pelvis
     DeltaComponentType=EulerY
     InputRange=(X=0,Y=40)
     TwistAxis="0,1,0"
    )) -> (Output_Get:double)
   > Branch(Condition=Get crouching() -> (crouching:bool))
   > Set thigh_rot_fwd_l(thigh_rot_fwd_l=0)
   > Set thigh_rot_out_l(thigh_rot_out_l=0)
   Set arm_up_r(
    execute=@n1.then_2
    arm_up_r:double=GetAlphaFromReferenceTransformDelta(
     SkeletalMeshComponent=Get Skeletal_Mesh()
     BoneName=upperarm_r
     BaseName=clavicle_r
     DeltaComponentType=QuaternionTwist
     InputRange=(X=-60,Y=-100)
     TwistAxis="0,0,1"
    )
   ) -> (Output_Get:double)
   n2 = GetAlphaFromReferenceTransformDelta(
    SkeletalMeshComponent=Get Skeletal_Mesh()
    BoneName=upperarm_r
    BaseName=clavicle_r
    DeltaComponentType=QuaternionTwist
    InputRange=(X=-100,Y=-120)
    TwistAxis="0,0,1"
   )
   > Branch(Condition="float > float"(A=@n2, B=0) {is=title K2Node_CallFunction})
   > Set arm_up_r(arm_up_r=float - float(B=@n2) {is=title K2Node_CallFunction})
   n3 = Get AnimInst() -> (AnimInst:Anim Instance Object Reference)
   Is Valid(exec=@n1.then_3, InputObject=@n3)
   > n4 = Branch(Condition="float > float"(A=GetCurveValue(self=@n3, CurveName=FrontEnd_Idle_Bypass), B=0))
   > Set FrontEnd_Idle_Float(FrontEnd_Idle_Float:double=1) -> (Output_Get:double)
   Set FrontEnd_Idle_Float(execute=@n4.else, FrontEnd_Idle_Float=0)
   Event OnEndCrouching()
   > Set crouching(crouching:bool=false) -> (Output_Get:bool)
   > Set crouching_alpha(crouching_alpha:double=0) -> (Output_Get:double)
   Event BlueprintInitializeAnimation()
   n5 = GetPartSkeletalMeshComponent()
   > Is Valid(InputObject=@n5)
   > Set Skeletal_Mesh(Skeletal_Mesh:Skeletal Mesh Component Object Reference=@n5) -> (Output_Get:Skeletal Mesh Component Object Reference)
   > n6 = Cast To FortPlayerPawn(Object=GetOwningActor())
   > Is Valid(InputObject=@n6)
   > n7 = Set Pawn(Pawn:Fort Player Pawn Object Reference=@n6) -> (Output_Get:Fort Player Pawn Object Reference)
   n8 = GetAnimInstance(self=Get Mesh(self=@n7))
   > Is Valid(InputObject=@n8)
   > Set AnimInst(AnimInst:Anim Instance Object Reference=@n8) -> (Output_Get:Anim Instance Object Reference)
   Event OnBeginCrouching()
   > Set crouching(crouching=true)
   > Set crouching_alpha(crouching_alpha=1)
  ]}
]
variables [
 AnimInst {type=AnimInstance, default=null}
 arm_up_r {type=number}
 crouching {type=boolean}
 crouching_alpha {type=number}
 FrontEnd_Idle_Float {type=number}
 Pawn {type=FortPlayerPawn}
 Skeletal_Mesh {type=SkeletalMeshComponent}
 thigh_rot_fwd_l {type=number}
 thigh_rot_fwd_r {type=number}
 thigh_rot_out_l {type=number}
 thigh_rot_out_r {type=number}
]
