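# Physics asset for the face accessory mesh of the medium male Faux player cosmetic, giving its dynamic hair bones simulated secondary motion.
# Small capsule bodies cover four hair bones (back center, back left, back right and tip) plus the head.
# Each hair bone is constrained directly to the head rather than chained, with tight swing and twist limits that keep movement subtle.
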PhysicsAsset /BRCosmetics/Characters/Player/Male/Medium/Bodies/M_MED_Faux/Meshes/Parts/M_MED_Faux_FaceAcc_Physics
bodies {
 dyn_hair_bckA [{shapeType=Capsule, length=0.5, radius=0.75, center=<0.4 0 0>, rotation=<rot 90 0 0>}]
 dyn_hair_bck_l [{shapeType=Capsule, length=1.4946, radius=0.7208, center=<0.22 0 0>, rotation=<rot 90 0 0>}]
 dyn_hair_bck_r [{shapeType=Capsule, length=1, radius=0.75, center=<0.35 0 0>, rotation=<rot 90 0 0>}]
 dyn_hair_tip [{shapeType=Capsule, length=1.005, radius=0.755, center=<0.22 0 0>, rotation=<rot 90 0 0>}]
 head [{shapeType=Capsule, length=1, radius=1}]
}
constraints [
 {bone1Name=dyn_hair_bckA, bone2Name=head, swing1LimitDegrees=5, swing2LimitDegrees=5, twistLimitDegrees=1}
 {bone1Name=dyn_hair_bck_r, bone2Name=head, swing1LimitDegrees=5, swing2LimitDegrees=10, twistLimitDegrees=1}
 {bone1Name=dyn_hair_bck_l, bone2Name=head, swing1LimitDegrees=5, swing2LimitDegrees=5, twistLimitDegrees=5}
 {bone1Name=dyn_hair_tip, bone2Name=head, swing1LimitDegrees=5, swing2LimitDegrees=5, twistLimitDegrees=1}
]
